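; PilgrimCore Blueprint function library of shared helpers for the Pilgrim music mode and its UI.
; MIDI note and text events convert to and from enums via DataTables whose row names are keys, using linear scans for reverse lookups.
; Other helpers resolve held-instrument track types, icons and names, test fade, Jam UI, touch-layout and controller-disconnect states, and lazy-load album art.
; GetWarmupMeshTypes maps a Material's usage flags to mesh-type enums.
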(fn CanTrackFadeIn
  (params (GemBreakEndPending :type Boolean) (IsInEndingGemBreak :type Boolean))
  (return (and GemBreakEndPending (not IsInEndingGemBreak))))

(fn CanTrackFadeOut
  (params (IsInGemBreak :type Boolean) (IsInStartingGemBreak :type Boolean))
  (return (and IsInGemBreak (not IsInStartingGemBreak))))

(fn CanTrackFadeOutAtSongEnd
  (params (SongStopped :type Boolean) (IsInGemBreak :type Boolean))
  (return (and SongStopped (not IsInGemBreak))))

(fn GetAllEventsMidiNoteEnums
  (bind _outrownames (DataTable|GetDataTableRowNames "/PilgrimCore/Utils/MidiEvents/Events/DT_PilgrimEventsMidiNoteEvents.DT_PilgrimEventsMidiNoteEvents"))
  (for _array_element _outrownames
    (bind _returnvalue (Utilities|GetDataTableRow "/PilgrimCore/Utils/MidiEvents/Events/DT_PilgrimEventsMidiNoteEvents.DT_PilgrimEventsMidiNoteEvents" _array_element))
    (Utilities|Array|Add (LocalVariables|GetValues) (Utilities|Struct|BreakSPilgrimEventsMidiNoteEventsTableRow _returnvalue)))
  ;[1 0.5 0.5] Return all values found in the table
  (return (LocalVariables|GetValues)))

(fn GetAllEventsMidiNoteIntegers
  ;[1 0.5 0.5] Return all values found in the table
  (return (CallFunction|GetAllIntegerKeysfromTable "/PilgrimCore/Utils/MidiEvents/Events/DT_PilgrimEventsMidiNoteEvents.DT_PilgrimEventsMidiNoteEvents")))

(fn GetAllEventsMidiTextEnums
  (bind _outrownames (DataTable|GetDataTableRowNames "/PilgrimCore/Utils/MidiEvents/Events/DT_PilgrimEventsMidiTextEvents.DT_PilgrimEventsMidiTextEvents"))
  (for _array_element _outrownames
    (bind _returnvalue (Utilities|GetDataTableRow "/PilgrimCore/Utils/MidiEvents/Events/DT_PilgrimEventsMidiTextEvents.DT_PilgrimEventsMidiTextEvents" _array_element))
    (Utilities|Array|Add (LocalVariables|GetValues) (Utilities|Struct|BreakSPilgrimEventsMidiTextEventsTableRow _returnvalue)))
  ;[1 0.5 0.5] Return all values found in the table
  (return (LocalVariables|GetValues)))

(fn GetAllEventsMidiTextStrings
  ;[1 0.5 0.5] Return all values found in the table
  (return (CallFunction|GetAllStringKeysfromTable "/PilgrimCore/Utils/MidiEvents/Events/DT_PilgrimEventsMidiTextEvents.DT_PilgrimEventsMidiTextEvents")))

(fn GetAllIntegerKeysFromTable
  (params (DataTable :type Data%20Table%20Object%20Reference))
  (bind _outrownames (DataTable|GetDataTableRowNames DataTable))
  (for _array_element _outrownames
    (Utilities|Array|Add (LocalVariables|GetValues) (Utilities|String|StringToInteger (Utilities|String|ToString(Name) _array_element))))
  ;[1 0.5 0.5] Return all values found in the table
  (return (LocalVariables|GetValues)))

(fn GetAllStringKeysFromTable
  (params (DataTable :type Data%20Table%20Object%20Reference))
  (bind _outrownames (DataTable|GetDataTableRowNames DataTable))
  (for _array_element _outrownames
    (Utilities|Array|Add (LocalVariables|GetValues) (Utilities|String|ToString(Name) _array_element)))
  ;[1 0.5 0.5] Return all values found in the table
  (return (LocalVariables|GetValues)))

(fn GetEventsMidiNoteEnumFromInteger
  (params (MidiNoteEventInt :type Integer))
  (bind _returnvalue (Utilities|GetDataTableRow "/PilgrimCore/Utils/MidiEvents/Events/DT_PilgrimEventsMidiNoteEvents.DT_PilgrimEventsMidiNoteEvents" (Utilities|String|StringToName (Utilities|String|ToString(Integer) MidiNoteEventInt)))
    (:then
      ;[1 0.5 0.5] Return
      (return (Utilities|Struct|BreakSPilgrimEventsMidiNoteEventsTableRow _returnvalue)))
    (:RowNotFound
      ;[1 0 0] Error: Could not find row
      (Development|PrintString 0 (Utilities|String|Append "Could not find row for Midi Note: " (Utilities|String|ToString(Integer) (Variables|GetMidiNoteEventInt))) true true "(R=0.000000,G=0.660000,B=1.000000,A=1.000000)" 5.0)
      (return "NewEnumerator3"))))

(fn GetEventsMidiNoteEnumsFromIntegers
  (params (MidiNoteEventInt :type Array%20of%20Integers))
  (for _array_element MidiNoteEventInt
    (Utilities|Array|Add (LocalVariables|GetValues) (MidiEvents|NoteEvents|Events|GetEventsMidiNoteEnumfromInteger _array_element)))
  ;[1 0.5 0.5] Return
  (return (LocalVariables|GetValues)))

(fn GetEventsMidiNoteIntegerFromEnum
  (params (MidiNoteEventEnum :type EPilgrimEventsMidiNoteEvents%20Enum :default "NewEnumerator3"))
  (bind _outrownames (DataTable|GetDataTableRowNames "/PilgrimCore/Utils/MidiEvents/Events/DT_PilgrimEventsMidiNoteEvents.DT_PilgrimEventsMidiNoteEvents"))
  (for _array_element _outrownames
    (bind _returnvalue (Utilities|GetDataTableRow "/PilgrimCore/Utils/MidiEvents/Events/DT_PilgrimEventsMidiNoteEvents.DT_PilgrimEventsMidiNoteEvents" _array_element)
      (:then
        (if (Utilities|Enum|Equal(Enum) (Utilities|Struct|BreakSPilgrimEventsMidiNoteEventsTableRow _returnvalue) MidiNoteEventEnum)
          ;[1 0.5 0.5] Return the corresponding int
          (return (Utilities|String|StringToInteger (Utilities|String|ToString(Name) _array_element)))))
      (:RowNotFound
        ;[1 0 0] Error: Could not find row
        (Development|PrintString 0 (Utilities|String|Append "Could not find row for Midi Note: " (Utilities|String|EnumtoString (Variables|GetMidiNoteEventEnum))) true true "(R=0.000000,G=0.660000,B=1.000000,A=1.000000)" 5.0)
        (return 0)))))

(fn GetEventsMidiNoteIntegersFromEnums
  (params (MidiNoteEventEnum :type Array%20of%20EPilgrimEventsMidiNoteEvents%20Enums :default "NewEnumerator3"))
  (for _array_element MidiNoteEventEnum
    (Utilities|Array|Add (LocalVariables|GetValues) (MidiEvents|NoteEvents|Events|GetEventsMidiNoteIntegerfromEnum _array_element)))
  ;[1 0.5 0.5] Return the corresponding int
  (return (LocalVariables|GetValues)))

(fn GetEventsMidiTextEnumFromString
  (params (MidiTextEventString :type String))
  (bind _returnvalue (Utilities|GetDataTableRow "/PilgrimCore/Utils/MidiEvents/Events/DT_PilgrimEventsMidiTextEvents.DT_PilgrimEventsMidiTextEvents" (Utilities|String|StringToName MidiTextEventString))
    (:then
      ;[1 0.5 0.5] Return
      (return (Utilities|Struct|BreakSPilgrimEventsMidiTextEventsTableRow _returnvalue)))
    (:RowNotFound
      ;[1 0 0] Error: Could not find row
      (Development|PrintString 0 (Utilities|String|Append "Could not find row for Midi Text: " (Variables|GetMidiTextEventString)) true true "(R=0.000000,G=0.660000,B=1.000000,A=1.000000)" 5.0)
      (return "NewEnumerator0"))))

(fn GetEventsMidiTextEnumsFromStrings
  (params (MidiTextEventString :type Array%20of%20Strings))
  (for _array_element MidiTextEventString
    (Utilities|Array|Add (LocalVariables|GetValues) (MidiEvents|TextEvents|Events|GetEventsMidiTextEnumfromString _array_element)))
  ;[1 0.5 0.5] Return
  (return (LocalVariables|GetValues)))

(fn GetEventsMidiTextStringFromEnum
  (params (MidiTextEventEnum :type EPilgrimEventsMidiTextEvents%20Enum :default "NewEnumerator0"))
  (bind _outrownames (DataTable|GetDataTableRowNames "/PilgrimCore/Utils/MidiEvents/Events/DT_PilgrimEventsMidiTextEvents.DT_PilgrimEventsMidiTextEvents"))
  (for _array_element _outrownames
    (bind _returnvalue (Utilities|GetDataTableRow "/PilgrimCore/Utils/MidiEvents/Events/DT_PilgrimEventsMidiTextEvents.DT_PilgrimEventsMidiTextEvents" _array_element)
      (:then
        (if (Utilities|Enum|Equal(Enum) (Utilities|Struct|BreakSPilgrimEventsMidiTextEventsTableRow _returnvalue) MidiTextEventEnum)
          ;[1 0.5 0.5] Return the corresponding int
          (return (Utilities|String|ToString(Name) _array_element))))
      (:RowNotFound
        ;[1 0 0] Error: Could not find row
        (Development|PrintString 0 (Utilities|String|Append "Could not find row for Midi Text: " (Utilities|String|EnumtoString (Variables|GetMidiTextEventEnum))) true true "(R=0.000000,G=0.660000,B=1.000000,A=1.000000)" 5.0)
        (return "")))))

(fn GetEventsMidiTextStringsFromEnums
  (params (MidiTextEventEnum :type Array%20of%20EPilgrimEventsMidiTextEvents%20Enums :default "NewEnumerator0"))
  (for _array_element MidiTextEventEnum
    (Utilities|Array|Add (LocalVariables|GetValues) (MidiEvents|TextEvents|Events|GetEventsMidiTextStringfromEnum _array_element)))
  ;[1 0.5 0.5] Return the corresponding int
  (return (LocalVariables|GetValues)))

(fn GetHeldInstrumentTrackType
  (params (Pawn :type Pawn%20Object%20Reference))
  (bind _asfort_player_pawn (Utilities|Casting|CastToFortPlayerPawn Pawn)
    (:then
      (Utilities|IsValid (Actor|GetComponentbyClass (Variables|/Script/FortniteGame.FortPawn|GetCurrentWeapon _asfort_player_pawn) "/SparksCosmetics/Components/BP_SparksInstrumentSettingsComponent.BP_SparksInstrumentSettingsComponent_C")
        (:Is%20Valid
          (return (Class|PilgrimCoreBPFL|SparksInstrumenttoPilgrimTrackType (Variables|/SparksCosmetics/Components/BP_SparksInstrumentSettingsComponent.BP_SparksInstrumentSettingsComponent_C|GetInstrumentType (Actor|GetComponentbyClass (Variables|/Script/FortniteGame.FortPawn|GetCurrentWeapon _asfort_player_pawn) "/SparksCosmetics/Components/BP_SparksInstrumentSettingsComponent.BP_SparksInstrumentSettingsComponent_C")))))
        (:Is%20Not%20Valid
          _)))
    (:CastFailed
      _))
  (return "TrackNone"))

(fn GetInstrumentIconIndex
  (params (InstrumentType :type EPilgrimTrackType%20Enum :default "TrackGuitar") (SongMetaData :type Pilgrim%20Song%20Metadata%20Object%20Reference) (StartingInstrumentIndex :type ESparksInstrumentType%20Enum :default "Guitar") (bUseKeyTarIcon :type Boolean :default true))
  (LocalVariables|SetbUseKeyTar bUseKeyTarIcon)
  (LocalVariables|SetInstrument%20Type InstrumentType)
  (Utilities|IsValid SongMetaData
    (:Is%20Valid
      (LocalVariables|SetCurrentInstrument (Class|PilgrimSongMetadata|GetStartingInstrumentType SongMetaData (LocalVariables|GetInstrument%20Type))))
    (:Is%20Not%20Valid
      (LocalVariables|SetCurrentInstrument StartingInstrumentIndex)))
  (return (Utilities|Select :Index (LocalVariables|GetCurrentInstrument) :Guitar (select (Utilities|Enum|Equal(Enum) (LocalVariables|GetInstrument%20Type) "TrackPlasticGuitar") 4 0) :Bass (select (Utilities|Enum|Equal(Enum) (LocalVariables|GetInstrument%20Type) "TrackPlasticBass") 7 1) :Vocals (select (Utilities|Enum|Equal(Enum) (LocalVariables|GetInstrument%20Type) "TrackProVocals") 10 2) :Drum (select (Utilities|Enum|Equal(Enum) (LocalVariables|GetInstrument%20Type) "TrackPlasticDrum") 9 (select (Utilities|Enum|Equal(Enum) (LocalVariables|GetInstrument%20Type) "TrackProCymbals") 8 3)) :Keyboard (select (and (or (Utilities|Enum|Equal(Enum) (LocalVariables|GetInstrument%20Type) "TrackPlasticGuitar") (Utilities|Enum|Equal(Enum) (LocalVariables|GetInstrument%20Type) "TrackPlasticBass")) (LocalVariables|GetbUseKeyTar)) 5 6) :None 0)))

(fn GetInstrumentNameFromPart
  (params (TrackType :type EPilgrimTrackType%20Enum :default "TrackGuitar"))
  (return (Utilities|Select :Index TrackType :TrackGuitar "LOCTABLE(\"/SparksSettings/UI/ST_Sparks_InstrumentNames.ST_Sparks_InstrumentNames\", \"Lead\")" :TrackBass "LOCTABLE(\"/SparksSettings/UI/ST_Sparks_InstrumentNames.ST_Sparks_InstrumentNames\", \"Bass\")" :TrackVocals "LOCTABLE(\"/SparksSettings/UI/ST_Sparks_InstrumentNames.ST_Sparks_InstrumentNames\", \"Vocals\")" :TrackDrum "LOCTABLE(\"/SparksSettings/UI/ST_Sparks_InstrumentNames.ST_Sparks_InstrumentNames\", \"Drums\")" :TrackPlasticGuitar "LOCTABLE(\"/SparksSettings/UI/ST_Sparks_InstrumentNames.ST_Sparks_InstrumentNames\", \"ProLead\")" :TrackPlasticBass "LOCTABLE(\"/SparksSettings/UI/ST_Sparks_InstrumentNames.ST_Sparks_InstrumentNames\", \"ProBass\")" :TrackPlasticDrum "LOCTABLE(\"/SparksSettings/UI/ST_Sparks_InstrumentNames.ST_Sparks_InstrumentNames\", \"ProDrums\")" :TrackProVocals "LOCTABLE(\"/SparksSettings/UI/ST_Sparks_InstrumentNames.ST_Sparks_InstrumentNames\", \"ProVocals\")" :TrackProCymbals "LOCTABLE(\"/SparksSettings/UI/ST_Sparks_InstrumentNames.ST_Sparks_InstrumentNames\", \"ProDrumsCymbals\")")))

(fn GetWarmupMeshTypes
  (params (Material :type Material%20Object%20Reference))
  (bind _meshtypestemp (LocalVariables|GetMeshTypesTemp))
  (Utilities|FlowControl|Sequence
    (:then_0
      ;
      (if (Variables|/Script/Engine.Material|GetbUsedWithSkeletalMesh Material)
        (Utilities|Array|AddUnique _meshtypestemp "NewEnumerator0")))
    (:then_1
      (if (Variables|/Script/Engine.Material|GetbUsedWithMeshDeformer Material)
        (Utilities|Array|AddUnique _meshtypestemp "NewEnumerator1")))
    (:then_2
      (if (Variables|/Script/Engine.Material|GetbUsedWithEditorCompositing Material)
        (Utilities|Array|AddUnique _meshtypestemp "NewEnumerator2")))
    (:then_3
      (if (Variables|/Script/Engine.Material|GetbUsedWithParticleSprites Material)
        (Utilities|Array|AddUnique _meshtypestemp "NewEnumerator3")))
    (:then_4
      (if (Variables|/Script/Engine.Material|GetbUsedWithBeamTrails Material)
        (Utilities|Array|AddUnique _meshtypestemp "NewEnumerator4")))
    (:then_5
      (if (Variables|/Script/Engine.Material|GetbUsedWithMeshParticles Material)
        (Utilities|Array|AddUnique _meshtypestemp "NewEnumerator5")))
    (:then_6
      (if (Variables|/Script/Engine.Material|GetbUsedWithNiagaraSprites Material)
        (Utilities|Array|AddUnique _meshtypestemp "NewEnumerator6")))
    (:then_7
      (if (Variables|/Script/Engine.Material|GetbUsedWithNiagaraRibbons Material)
        (Utilities|Array|AddUnique _meshtypestemp "NewEnumerator7")))
    (:then_8
      (if (Variables|/Script/Engine.Material|GetbUsedWithNiagaraMeshParticles Material)
        (Utilities|Array|AddUnique _meshtypestemp "NewEnumerator8")))
    (:then_9
      (if (Variables|/Script/Engine.Material|GetbUsedWithGeometryCache Material)
        (Utilities|Array|AddUnique _meshtypestemp "NewEnumerator9")))
    (:then_10
      (if (Variables|/Script/Engine.Material|GetbUsedWithMorphTargets Material)
        (Utilities|Array|AddUnique _meshtypestemp "NewEnumerator10")))
    (:then_11
      (if (Variables|/Script/Engine.Material|GetbUsedWithSplineMeshes Material)
        (Utilities|Array|AddUnique _meshtypestemp "NewEnumerator11")))
    (:then_12
      (if (Variables|/Script/Engine.Material|GetbUsedWithInstancedStaticMeshes Material)
        (Utilities|Array|AddUnique _meshtypestemp "NewEnumerator12")))
    (:then_13
      (if (Variables|/Script/Engine.Material|GetbUsedWithGeometryCollections Material)
        (Utilities|Array|AddUnique _meshtypestemp "NewEnumerator13")))
    (:then_14
      (if (Variables|/Script/Engine.Material|GetbUsedWithClothing Material)
        (Utilities|Array|AddUnique _meshtypestemp "NewEnumerator15")))
    (:then_15
      (if (Variables|/Script/Engine.Material|GetbUsedWithWater Material)
        (Utilities|Array|AddUnique _meshtypestemp "NewEnumerator16")))
    (:then_16
      (if (Variables|/Script/Engine.Material|GetbUsedWithHairStrands Material)
        (Utilities|Array|AddUnique _meshtypestemp "NewEnumerator17")))
    (:then_17
      (if (Variables|/Script/Engine.Material|GetbUsedWithLidarPointCloud Material)
        (Utilities|Array|AddUnique _meshtypestemp "NewEnumerator18")))
    (:then_18
      (if (Variables|/Script/Engine.Material|GetbUsedWithVirtualHeightfieldMesh Material)
        (Utilities|Array|AddUnique _meshtypestemp "NewEnumerator19")))
    (:then_19
      (if (Variables|/Script/Engine.Material|GetbUsedWithStaticMesh Material)
        (Utilities|Array|AddUnique _meshtypestemp "NewEnumerator24")))
    (:then_20
      (if (Variables|/Script/Engine.Material|GetbUsedWithVolumetricCloud Material)
        (Utilities|Array|AddUnique _meshtypestemp "NewEnumerator22")))
    (:then_21
      (if (Variables|/Script/Engine.Material|GetbUsedWithNanite Material)
        (Utilities|Array|AddUnique _meshtypestemp "NewEnumerator20")))
    (:then_22
      (if (Variables|/Script/Engine.Material|GetbUsedWithVoxels Material)
        (Utilities|Array|AddUnique _meshtypestemp "NewEnumerator21")))
    (:then_23
      (return (LocalVariables|GetMeshTypesTemp)))))

(fn HasTrackPart
  (params (SongMetadata :type Pilgrim%20Song%20Metadata%20Object%20Reference) (TrackType :type EPilgrimTrackType%20Enum :default "TrackGuitar"))
  (return (!= (Meta|GetMissingTrackIntensity) (Class|PilgrimSongMetadata|GetTrackIntensity SongMetadata TrackType))))

(fn IsJamShowingNowPlayingUI
  (if (GameplayTags|DoesContainerMatchTagQuery (Class|FortGameStateAthena|GetAthenaPlaylistContextTags (Fortnite|Athena|GameState|GetGameStateAthena 0)) (GameplayTags|MakeGameplayTagQuery "(TokenStreamVersion=0,TagDictionary=((TagName=\"FMJam.EnableNowPlayingUI\")),QueryTokenStream=(0,1,1,1,0),UserDescription=\"\",AutoDescription=\" ANY( FMJam.EnableNowPlayingUI )\")"))
    (return true)
    (else
      (bind _returnvalue (Development|GetConsoleVariableBoolValue "FMJam.ShowQuickMix"))
      (if _returnvalue
        (bind _returnvalue_2 (Development|GetConsoleVariableBoolValue "FMJam.ShowNowPlayingWithQuickMix"))
        (return _returnvalue_2)
        (else
          (return false))))))

(fn IsPilgrimInverse
  (params (OwningPlayerController :type Player%20Controller%20Object%20Reference))
  (return (and (DisplayHelper|IsUsingCompactUI) (Inputs|IsUsingInputType "Touch" OwningPlayerController))))

(fn IsPlayerControllerDisconnected
  (params (Owning%20Player :type Player%20Controller%20Object%20Reference) (DeviceConnectionChanged :type Pilgrim%20Event%20Player%20State%20Player%20Input%20Device%20Connection%20Changed%20Structure))
  (bind (_playercontroller _newconnectionstate _remainingconnecteddevicesforuser) (Utilities|Struct|BreakPilgrimEventPlayerStatePlayerInputDeviceConnectionChanged DeviceConnectionChanged))
  (return (== Owning%20Player _playercontroller) (and (Utilities|Enum|NotEqual(Enum) _newconnectionstate "Connected") (== _remainingconnecteddevicesforuser 0))))

(fn IsPlayersControllersAllDisconnected
  (params (Player :type Player%20Controller%20Object%20Reference))
  (bind _returnvalue (Actor|GetComponentbyClass Player "/Script/PilgrimCoreRuntime.PilgrimPlayerControllerComponent"))
  (Utilities|IsValid _returnvalue
    (:Is%20Valid
      (return (Class|PilgrimPlayerControllerComponent|IsAllPlayerControllerDevicesDisconnected _returnvalue)))
    (:Is%20Not%20Valid
      (return false))))

(fn LazyLoadAlbumArt
  (params (Song%20Metadata :type Pilgrim%20Song%20Metadata%20Object%20Reference) (Lazy%20Image :type Fort%20Lazy%20Image%20Object%20Reference))
  (Class|FortLazyImage|SetEnableTextureReleasingOverride Lazy%20Image true)
  (SparksAlbumArt|RequestLazyAlbumArt (GameInstanceSubsystems|GetSparksAlbumArtManager 0) (Meta|GetShortName Song%20Metadata) Lazy%20Image)
  (return))
